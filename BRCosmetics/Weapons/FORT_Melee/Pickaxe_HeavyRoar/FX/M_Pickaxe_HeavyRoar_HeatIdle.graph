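# Additive unlit material for Niagara mesh particles, named as the idle heat effect of the HeavyRoar pickaxe.
# Particle color RGB feeds the shared MF_FX_EmissiveAdjustments function, with no textures sampled.
# Particle alpha is multiplied by a static-switched UV Mask, either an inverted V-coordinate gradient or a constant one.
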Material /BRCosmetics/Weapons/FORT_Melee/Pickaxe_HeavyRoar/FX/M_Pickaxe_HeavyRoar_HeatIdle
graph {
 properties {
  bUsedWithNiagaraMeshParticles=true
  blendMode=BLEND_Additive
  pixelDepthOffsetMode=PDOM_Legacy
  shadingModel=MSM_Unlit
  usedShadingModels=Unlit
 }
 nodes [
  n1 = ParticleColor()
  MaterialFunctionCall(Emissive=@n1.RGB) {materialFunction=&MF_FX_EmissiveAdjustments}
  n2 = ComponentMask(None=TextureCoordinate()) {g=true}
  Multiply(A=@n1.A, B=StaticSwitchParameter(True=OneMinus(None=@n2), False=Constant() {r=1}) {parameterName=UV Mask})
 ]
}
