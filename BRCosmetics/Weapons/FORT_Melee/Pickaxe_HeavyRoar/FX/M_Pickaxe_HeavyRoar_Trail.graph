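# Translucent unlit Niagara ribbon material for the trail of the HeavyRoar pickaxe.
# A trail texture borrowed from the Future Samurai Summer pickaxe pans along UV1, shifted by a dynamic random offset.
# A UV0 gradient remapped by dynamic Length X, minus a steep power falloff, shapes trail length and end fade.
# Particle color drives emissive and alpha, with view-angle fading on the mask and a camera-direction offset term.
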Material /BRCosmetics/Weapons/FORT_Melee/Pickaxe_HeavyRoar/FX/M_Pickaxe_HeavyRoar_Trail
graph {
 properties {bUsedWithNiagaraRibbons=true, blendMode=BLEND_Translucent, pixelDepthOffsetMode=PDOM_Legacy, shadingModel=MSM_Unlit, usedShadingModels=Unlit}
 nodes [
  n1 = ParticleColor()
  MaterialFunctionCall(Emissive=@n1.RGB) {materialFunction=&MF_FX_EmissiveAdjustments}
  Multiply(
   A=MaterialFunctionCall() {materialFunction=&CameraDirectionVector}.Result
   B=ScalarParameter() {parameterName=Camera Offset}
  )
  n2 = TextureCoordinate() {coordinateIndex=1}
  n3 = AppendVector(A=DynamicParameter() {defaultValue=<rgba 0 1 1 1>, paramNames [Random Offset X, Length X, Param3, Param4]}.Random Offset X, B=Constant())
  n4 = ScalarParameter() {defaultValue=0.1, parameterName=Speed X, sliderMax=1}
  n5 = Constant()
  n6 = TextureSample(UVs=Panner(Coordinate=Add(A=@n2, B=@n3), Speed=AppendVector(A=@n4, B=@n5)) {constCoordinate=1}) {
   constCoordinate=1
   texture=/BRCosmetics/Weapons/FORT_Melee/Pickaxe_Future_Samurai_Summer_Male/FX/T_Pickaxe_Future_Samurai_Summer_Male_Trail_01
  }
  n7 = OneMinus(None=ComponentMask(None=TextureCoordinate()) {r=true})
  n8 = Power(Base=@n7) {constExponent=1}
  n9 = OneMinus(None=DynamicParameter() {defaultValue=<rgba 0 1 1 1>, paramNames [Random Offset X, Length X, Param3, Param4]}.Length X)
  n10 = Subtract(A=Divide(A=Subtract(A=@n8, B=@n9), B=OneMinus(None=@n9)), B=Power(Base=@n7) {constExponent=10})
  Multiply(
   A=Multiply(A=Multiply(A=@n6.R, B=@n10), B=@n1.A)
   B=MaterialFunctionCall("FadeSensitivity (optional)"=ScalarParameter() {defaultValue=5, parameterName=FadeSensitivity, sliderMax=20}) {materialFunction=&FadeBasedOnViewAngle}.Result
  )
 ]
}
